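# Material function in the FNE_UI effects library that slants (shears) 2D UV coordinates for UI materials.
# Swizzles cross the axes so each slant component offsets its own axis in proportion to the other UV coordinate, added to the original UVs.
# Subtracting half the slant amount keeps the shear centered on the middle of UV space.
# Defaults slant only horizontally, and the description "Linear Time" does not match the graph.
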MaterialFunction /FNE_UI/Systems/_MaterialFunctions/Effects/MF_FNEUI_Slant
graph {
 properties {description=Linear Time, libraryCategoriesText [""]}
 nodes [
  n1 = FunctionInput(Preview=TextureCoordinate()) {bUsePreviewValueAsDefault=true, inputName=UVs, inputType=FunctionInput_Vector2}
  n2 = FunctionInput(Preview=Constant2Vector() {r=0.25}) {bUsePreviewValueAsDefault=true, inputName=Slant amount, inputType=FunctionInput_Vector2, sortPriority=1}
  n3 = MaterialFunctionCall(XY=Multiply(A=@n1, B=MaterialFunctionCall(XY=@n2) {materialFunction=&Swizzle}.YX)) {materialFunction=&Swizzle}
  FunctionOutput(None=Subtract(A=Add(A=@n1, B=@n3.YX), B=Divide(A=@n2))) {outputName=Slanted UVs}
 ]
}
